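# Sound cue for the drop sound of the Precision Pistol weapon.
# A single wave player feeds a Modulator at default settings, which is the cue's root.
# It uses weapon-drop attenuation, the loot-collect sound class, a slightly reduced volume and a shared stop-oldest weapon-drop concurrency set.
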SoundCue /CorruptionItems/Sounds/Weapons/Pistol/PrecisionPistol/Pistol_Precision_Drop_Cue
root=Modulator_0
nodes [
 {id=Modulator_0, children [WavePlayer_3]}
 {id=WavePlayer_3, properties {soundWaveAssetPtr=&/Precision_Pistol_Drop}}
]
properties {
 attenuationSettings=&Fort_Weapon_Drop_Att
 maxDistance=3000
 soundClassObject=&Fort_Loot_Collect
 volumeMultiplier=0.9
 concurrencyOverrides {resolutionRule=PreventNew}
 concurrencySet [&Weapon_Drop_StopOldest]
}
