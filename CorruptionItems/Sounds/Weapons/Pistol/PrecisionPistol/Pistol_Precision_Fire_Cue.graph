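# Sound cue for the precision pistol's gunfire, using third-person and distant weapon-fire sound classes.
# Close, medium and far layers each pick randomly from five recordings, then pass through a modulator, their own distance attenuation and a sound class override.
# Close and medium are mixed, with close at half volume, then modulated again and mixed with the quieter distant layer.
# Gunfire concurrency sets manage overlapping shots, and the debug flag is left on.
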SoundCue /CorruptionItems/Sounds/Weapons/Pistol/PrecisionPistol/Pistol_Precision_Fire_Cue
root=Mixer_2
nodes [
 {id=Attenuation_3, children [Modulator_1], properties {attenuationSettings=&Fort_Weapon_Projectile_Close_Athena_Att}}
 {
  id=Attenuation_4
  children [Modulator_0]
  properties {attenuationSettings=&Fort_Weapon_Projectile_Distant_Mono_Athena_Att}
 }
 {
  id=Attenuation_5
  children [Modulator_2]
  properties {attenuationSettings=&Fort_Weapon_Projectile_Medium_Mono_Athena_Att}
 }
 {id=Mixer_2, children [Modulator_3,SoundClass_4], properties {inputVolume [1,0.4]}}
 {id=Mixer_3, children [SoundClass_3,SoundClass_5], properties {inputVolume [0.5,1]}}
 {id=Modulator_0, children [Random_5], properties {volumeMax=1, volumeMin=0.9}}
 {id=Modulator_1, children [Random_3], properties {volumeMax=1}}
 {id=Modulator_2, children [Random_4], properties {volumeMax=1, volumeMin=0.9}}
 {id=Modulator_3, children [Mixer_3]}
 {id=Random_3, children [WavePlayer_3, WavePlayer_4, WavePlayer_5, WavePlayer_6, WavePlayer_13], properties {weights [1,1,1,1,1]}}
 {id=Random_4, children [WavePlayer_14, WavePlayer_15, WavePlayer_16, WavePlayer_17, WavePlayer_18], properties {weights [1,1,1,1,1]}}
 {id=Random_5, children [WavePlayer_22, WavePlayer_23, WavePlayer_24, WavePlayer_25, WavePlayer_26], properties {weights [1,1,1,1,1]}}
 {
  id=SoundClass_3
  children [Attenuation_3]
  properties {soundClassOverride=&Fort_WeaponRanged_Fire_3P_PSM}
 }
 {
  id=SoundClass_4
  children [Attenuation_4]
  properties {soundClassOverride=&Fort_WeaponRanged_Fire_Distant}
 }
 {
  id=SoundClass_5
  children [Attenuation_5]
  properties {soundClassOverride=&Fort_WeaponRanged_Fire_3P_PSM}
 }
 {id=WavePlayer_13, properties {soundWaveAssetPtr=&/Precision_Pistol_P3_Fire_Close_05}}
 {id=WavePlayer_14, properties {soundWaveAssetPtr=&/Precision_Pistol_P3_Fire_Med_01}}
 {id=WavePlayer_15, properties {soundWaveAssetPtr=&/Precision_Pistol_P3_Fire_Med_02}}
 {id=WavePlayer_16, properties {soundWaveAssetPtr=&/Precision_Pistol_P3_Fire_Med_03}}
 {id=WavePlayer_17, properties {soundWaveAssetPtr=&/Precision_Pistol_P3_Fire_Med_04}}
 {id=WavePlayer_18, properties {soundWaveAssetPtr=&/Precision_Pistol_P3_Fire_Med_05}}
 {id=WavePlayer_22, properties {soundWaveAssetPtr=&/Precision_Pistol_P3_Fire_Far_05}}
 {id=WavePlayer_23, properties {soundWaveAssetPtr=&/Precision_Pistol_P3_Fire_Far_01}}
 {id=WavePlayer_24, properties {soundWaveAssetPtr=&/Precision_Pistol_P3_Fire_Far_02}}
 {id=WavePlayer_25, properties {soundWaveAssetPtr=&/Precision_Pistol_P3_Fire_Far_03}}
 {id=WavePlayer_26, properties {soundWaveAssetPtr=&/Precision_Pistol_P3_Fire_Far_04}}
 {id=WavePlayer_3, properties {soundWaveAssetPtr=&/Precision_Pistol_P3_Fire_Close_01}}
 {id=WavePlayer_4, properties {soundWaveAssetPtr=&/Precision_Pistol_P3_Fire_Close_02}}
 {id=WavePlayer_5, properties {soundWaveAssetPtr=&/Precision_Pistol_P3_Fire_Close_03}}
 {id=WavePlayer_6, properties {soundWaveAssetPtr=&/Precision_Pistol_P3_Fire_Close_04}}
]
properties {
 bDebug=true
 maxDistance=27500
 soundClassObject=&Fort_WeaponRanged_Fire_3P_PSM
 volumeMultiplier=0.9
 concurrencyOverrides {bLimitToOwner=true, maxCount=3, resolutionRule=StopOldest}
 concurrencySet [&Weapon_Gunfire_Global_Conc, &Weapon_Gunfire_Local_Conc]
}
