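# Sound cue for the Precision Pistol landing, classed under the loot collect sound class.
# A single wave player feeds the landed sound through a modulator left at its defaults.
# Cue-level attenuation and concurrency overrides give it a short falloff range and cap simultaneous instances, blocking new ones once the limit is reached.
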SoundCue /CorruptionItems/Sounds/Weapons/Pistol/PrecisionPistol/Pistol_Precision_Landed_Cue
root=Modulator_0
nodes [
 {id=Modulator_0, children [WavePlayer_0]}
 {id=WavePlayer_0, properties {soundWaveAssetPtr=&/Precision_Pistol_Landed}}
]
properties {
 bOverrideAttenuation=true
 bOverrideConcurrency=true
 maxDistance=1000
 soundClassObject=&Fort_Loot_Collect
 volumeMultiplier=1
 attenuationOverrides {falloffDistance=1000, attenuationShapeExtents {x=0}}
 concurrencyOverrides {maxCount=3, resolutionRule=PreventNew}
}
