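# Sound cue for the low-ammo warning of the Precision Pistol in the CorruptionItems content.
# A single wave player feeds the root directly, playing a shared Fortnite low-ammo rifle sample with no randomization or modulation.
# It routes through the degraded ranged-weapon low-ammo sound class and overrides concurrency so new instances are blocked, rather than stopping existing ones, once the limit is reached.
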SoundCue /CorruptionItems/Sounds/Weapons/Pistol/PrecisionPistol/Pistol_Precision_Low_Ammo_Cue
root=WavePlayer_34
nodes [{id=WavePlayer_34, properties {soundWaveAssetPtr=/Game/Sounds/Fort_Weapons/LowAmmo/weapon_low_ammo_hi_rifle_05}}]
properties {
 bOverrideConcurrency=true
 maxDistance=2097152
 soundClassObject=/Game/Sounds/FortSoundClassesAndModes/Fort_WeaponRanged_LowAmmo_Degraded_PSM
 concurrencyOverrides {resolutionRule=PreventNew}
}
