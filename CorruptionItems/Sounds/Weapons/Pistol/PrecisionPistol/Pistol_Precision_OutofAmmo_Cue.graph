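# Sound cue for the precision pistol's out-of-ammo sound.
# It picks at random, with equal weight, from four shared rifle out-of-ammo waves, then slightly varies volume through a modulator while keeping pitch fixed.
# The cue plays in the third-person ranged weapon fire sound class at reduced volume, with an attenuation falloff override and a concurrency override allowing two instances, stopping the oldest.
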SoundCue /CorruptionItems/Sounds/Weapons/Pistol/PrecisionPistol/Pistol_Precision_OutofAmmo_Cue
root=Modulator_0
nodes [
 {id=Modulator_0, children [Random_0], properties {pitchMax=1, pitchMin=1, volumeMax=0.95, volumeMin=0.85}}
 {id=Random_0, children [WavePlayer_0,WavePlayer_1,WavePlayer_2,WavePlayer_3], properties {weights [1,1,1,1]}}
 {id=WavePlayer_0, properties {soundWaveAssetPtr=/Game/Sounds/Fort_Weapons/OutofAmmo/Rifle_OutofAmmo_01}}
 {id=WavePlayer_1, properties {soundWaveAssetPtr=/Game/Sounds/Fort_Weapons/OutofAmmo/Rifle_OutofAmmo_02}}
 {id=WavePlayer_2, properties {soundWaveAssetPtr=/Game/Sounds/Fort_Weapons/OutofAmmo/Rifle_OutofAmmo_03}}
 {id=WavePlayer_3, properties {soundWaveAssetPtr=/Game/Sounds/Fort_Weapons/OutofAmmo/Rifle_OutofAmmo_04}}
]
properties {
 bOverrideConcurrency=true
 maxDistance=2097152
 soundClassObject=&Fort_WeaponRanged_Fire_3P_PSM
 volumeMultiplier=0.55
 attenuationOverrides {falloffDistance=1600}
 concurrencyOverrides {maxCount=2, resolutionRule=StopOldest}
}
