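# Sound cue that plays the pickup sound for the Precision Pistol, classed as loot-collect audio.
# A quality-level switch sends its first two tiers and last two tiers to separate mixers, each boosting a similarly named pickup wave from a different asset path.
# Shared weapon-pickup attenuation and a pickup concurrency set govern playback, alongside override values for a single instance that prevents new plays.
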SoundCue /CorruptionItems/Sounds/Weapons/Pistol/PrecisionPistol/Pistol_Precision_Pickup_Cue
root=QualityLevel_0
nodes [
 {id=Mixer_0, children [WavePlayer_6,null], properties {inputVolume [1.2,1]}}
 {id=Mixer_1, children [WavePlayer_1,null], properties {inputVolume [1.2,1]}}
 {id=QualityLevel_0, children [Mixer_1,Mixer_1,Mixer_0,Mixer_0]}
 {id=WavePlayer_1, properties {soundWaveAssetPtr=&/Precision_Pistol_Pickup}}
 {id=WavePlayer_6, properties {soundWaveAssetPtr=&Precision_Pistol_Pickup}}
]
properties {
 attenuationSettings=&Fort_Weapon_Pickup_Att
 maxDistance=3500
 soundClassObject=&Fort_Loot_Collect
 volumeMultiplier=0.4615
 concurrencyOverrides {maxCount=1, resolutionRule=PreventNew}
 concurrencySet [&PickupMaximum_Oldest]
}
