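# Sound cue for the start of the precision pistol's reload.
# A Random node picks between two equally weighted reload-start wave variants, feeding a root Modulator that uses default settings.
# It plays as weapon foley with the shared weapon reload attenuation at a reduced volume, overrides concurrency, and has the debug flag left on.
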SoundCue /CorruptionItems/Sounds/Weapons/Pistol/PrecisionPistol/Pistol_Precision_Reload_Start_Cue
root=Modulator_0
nodes [
 {id=Modulator_0, children [Random_1]}
 {id=Random_1, children [WavePlayer_0,WavePlayer_1], properties {weights [1,1]}}
 {id=WavePlayer_0, properties {soundWaveAssetPtr=&/Precision_Pistol_Reload_Start_01}}
 {id=WavePlayer_1, properties {soundWaveAssetPtr=&/Precision_Pistol_Reload_Start_02}}
]
properties {
 attenuationSettings=&Fort_Weapon_Reload_Att
 bDebug=true
 bOverrideConcurrency=true
 maxDistance=2250
 soundClassObject=&Fort_WeaponFoley
 volumeMultiplier=0.376
}
